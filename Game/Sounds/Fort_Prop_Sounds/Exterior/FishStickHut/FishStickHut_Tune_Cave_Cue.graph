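# Music cue for the Fish Stick Hut cave exterior prop, endlessly looping Coral Buddies tunes as attenuated world audio in the music sound class.
# A delay node feeds a weighted random pick among five waves, three of them played with a loop count of two.
# Single-input mixers set each tune's volume.
# Overrides add distance low-pass filtering, listener focus, a distant reverb send and a concurrency cap that stops the quietest.
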SoundCue /Game/Sounds/Fort_Prop_Sounds/Exterior/FishStickHut/FishStickHut_Tune_Cave_Cue
root=Looping_0
nodes [
 {id=Delay_0, children [Random_0]}
 {id=Looping_0, children [Delay_0]}
 {id=Looping_1, children [WavePlayer_7], properties {bLoopIndefinitely=false, loopCount=2}}
 {id=Looping_2, children [WavePlayer_5], properties {bLoopIndefinitely=false, loopCount=2}}
 {id=Looping_4, children [WavePlayer_2], properties {bLoopIndefinitely=false, loopCount=2}}
 {id=Mixer_0, children [WavePlayer_6], properties {inputVolume [2]}}
 {id=Mixer_1, children [Looping_1], properties {inputVolume [0.6]}}
 {id=Mixer_2, children [Looping_2], properties {inputVolume [1.4]}}
 {id=Mixer_3, children [WavePlayer_4], properties {inputVolume [1]}}
 {id=Mixer_4, children [Looping_4], properties {inputVolume [1]}}
 {id=Random_0, children [Mixer_4,Mixer_3,Mixer_2,Mixer_0,Mixer_1], properties {weights [0.2,0.1,0.2,1,1]}}
 {id=WavePlayer_2, properties {soundWaveAssetPtr=&/MX_CoralBuddies_Tune_02}}
 {id=WavePlayer_4, properties {soundWaveAssetPtr=&/MX_CoralBuddies_Tune_04}}
 {id=WavePlayer_5, properties {soundWaveAssetPtr=&/MX_CoralBuddies_Tune_05}}
 {id=WavePlayer_6, properties {soundWaveAssetPtr=&/MX_CoralBuddies_Tune_06}}
 {id=WavePlayer_7, properties {soundWaveAssetPtr=&/MX_CoralBuddies_Tune_07}}
]
properties {
 bOverrideAttenuation=true
 bOverrideConcurrency=true
 maxDistance=4000
 soundClassObject=/Game/Sounds/FortSoundClassesAndModes/Fort_Music
 volumeMultiplier=0.6
 attenuationOverrides {
  bAttenuateWithLPF=true
  bEnableListenerFocus=true
  bEnableSubmixSends=true
  falloffDistance=3000
  lPFFrequencyAtMax=1500
  lPFFrequencyAtMin=22000
  lPFRadiusMax=3000
  lPFRadiusMin=1500
  nonFocusVolumeAttenuation=0.75
  nonSpatializedRadiusEnd=750
  nonSpatializedRadiusStart=1000
  occlusionInterpolationTime=2
  occlusionTraceChannel=ECC_GameTraceChannel11
  stereoSpread=3000
  attenuationShapeExtents {x=1000}
  submixSendSettings [{
    bEnableHPFCutoff=false
    bEnableLPFCutoff=false
    disableManualSendClamp=false
    hPFCutoff=20
    lPFCutoff=20000
    maxSendDistance=6000
    maxSendLevel=1
    minSendDistance=400
    minSendLevel=1
    sendLevel=0.2
    sendLevelControlMethod=Linear
    soundSubmix=&Fortnite_Reverb_Distant
   }]
 }
 concurrencyOverrides {maxCount=3, resolutionRule=StopQuietest}
}
