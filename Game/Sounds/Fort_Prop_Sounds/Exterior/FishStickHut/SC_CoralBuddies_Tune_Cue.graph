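# Sound cue for looping Coral Buddies music tunes, filed under the FishStickHut exterior prop sounds.
# A Random node picks among five equally weighted wave players, each looping a different tune.
# Playback goes through the Fort_Music sound class at reduced volume.
# The cue overrides attenuation, including occlusion settings, and limits concurrency to a few instances, stopping the quietest first.
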SoundCue /Game/Sounds/Fort_Prop_Sounds/Exterior/FishStickHut/SC_CoralBuddies_Tune_Cue
root=Random_0
nodes [
 {id=Random_0, children [WavePlayer_2,WavePlayer_4,WavePlayer_5,WavePlayer_6,WavePlayer_7], properties {weights [1,1,1,1,1]}}
 {id=WavePlayer_2, properties {bLooping=true, soundWaveAssetPtr=&/MX_CoralBuddies_Tune_02}}
 {id=WavePlayer_4, properties {bLooping=true, soundWaveAssetPtr=&/MX_CoralBuddies_Tune_04}}
 {id=WavePlayer_5, properties {bLooping=true, soundWaveAssetPtr=&/MX_CoralBuddies_Tune_05}}
 {id=WavePlayer_6, properties {bLooping=true, soundWaveAssetPtr=&/MX_CoralBuddies_Tune_06}}
 {id=WavePlayer_7, properties {bLooping=true, soundWaveAssetPtr=&/MX_CoralBuddies_Tune_07}}
]
properties {
 bOverrideAttenuation=true
 bOverrideConcurrency=true
 maxDistance=4600
 soundClassObject=/Game/Sounds/FortSoundClassesAndModes/Fort_Music
 volumeMultiplier=0.6
 attenuationOverrides {
  occlusionLowPassFilterFrequency=1000
  occlusionTraceChannel=ECC_GameTraceChannel11
  occlusionVolumeAttenuation=0.5
  attenuationShapeExtents {x=1000}
 }
 concurrencyOverrides {maxCount=3, resolutionRule=StopQuietest}
}
